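# Material function in the SafeZone folder that triplanar-projects a scaled, panning normal map using pre-skinned position by default.
# Three planar samples are blended by contrast-adjusted normal weights into XYZ, flat-top, XY and Z outputs, in world or tangent space.
# A high-quality switch substitutes WorldAlignedNormals_HighQuality results on three outputs, but that call's position and divisor inputs are unwired.
# A parallel RGB/float4 blend feeds no output.
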MaterialFunction /Figment_S07/Athena/SafeZone/Preskinned_WP_normal_Texture_ScalePan1
graph {
 properties {bExposeToLibrary=true, description=Tiles a texture in worldspace, libraryCategoriesText [Misc,Texturing]}
 nodes [
  Constant()
  n1 = FunctionInput(Preview=TextureObject() {texture=/Engine/EngineMaterials/DefaultDiffuse}) {inputName=TextureObject, inputType=FunctionInput_Texture2D}
  n2 = FunctionInput(Preview=VertexNormalWS()) {bUsePreviewValueAsDefault=true, inputName=World Space Normal, previewValue=<0 0 1 1>}
  n3 = Multiply(A=@n2, B=Constant() {r=1})
  n4 = ComponentMask(None=@n3) {b=true}
  n5 = FunctionInput() {
   bUsePreviewValueAsDefault=true
   inputName=ProjectionTransitionContrast
   inputType=FunctionInput_Scalar
   sortPriority=10
   previewValue=<1 0 0 1>
  }
  n6 = MaterialFunctionCall(In=Abs(None=@n4), Contrast=@n5) {materialFunction=&CheapContrast}
  n7 = ComponentMask(None=@n3) {r=true}
  n8 = MaterialFunctionCall(In=Abs(None=@n7), Contrast=@n5) {materialFunction=&CheapContrast}
  n9 = MaterialFunctionCall!(
   Negative World Position Divisor
   World Position
   Texture=@n1
   Z lerp Alpha=OneMinus(None=@n6.Result)
   WorldNormal
   X lerp alpha=OneMinus(None=@n8.Result)
  ) {materialFunction=/Engine/Functions/Engine_MaterialFunctions02/WorldAlignedNormals_HighQuality}
  n10 = PreSkinnedPosition()
  n11 = FunctionInput() {bUsePreviewValueAsDefault=true, description=Size of texture in worldspace units, inputName=TextureSize, previewValue=<64 64 64 6>}
  n12 = Divide(
   A=FunctionInput(Preview=VertexInterpolator(VS=@n10).PS) {bUsePreviewValueAsDefault=true, inputName=WorldPosition}
   B=Multiply(A=Abs(None=@n11)) {constB=-1}
  )
  n13 = Time()
  n14 = FunctionInput(Preview=Constant2Vector()) {bUsePreviewValueAsDefault=true, inputType=FunctionInput_Vector2}
  n15 = Panner(Coordinate=ComponentMask(None=@n12) {b=true, r=true}, Time=@n13, Speed=@n14)
  n16 = TextureSample(UVs=@n15, Tex=@n1) {samplerType=SAMPLERTYPE_Normal}
  n17 = Multiply(A=@n16.R) {constB=-1}
  n18 = Multiply(A=@n16.B, B=ComponentMask(None=@n3) {g=true})
  n19 = Panner(Coordinate=ComponentMask(None=@n12) {b=true, g=true}, Time=@n13, Speed=@n14)
  n20 = TextureSample(UVs=@n19, Tex=@n1) {samplerType=SAMPLERTYPE_Normal}
  n21 = Multiply(A=@n7, B=@n20.B)
  n22 = Multiply(A=@n20.R) {constB=-1}
  n23 = LinearInterpolate(
   A=AppendVector(A=AppendVector(A=@n17, B=@n18), B=Multiply(A=@n16.G) {constB=-1})
   B=AppendVector(A=AppendVector(A=@n21, B=@n22), B=Multiply(A=@n20.G) {constB=-1})
   Alpha=@n8.Result
  )
  n24 = Panner(Coordinate=ComponentMask(None=@n12) {g=true, r=true}, Time=@n13, Speed=@n14)
  n25 = TextureSample(UVs=@n24, Tex=@n1) {samplerType=SAMPLERTYPE_Normal}
  n26 = Constant3Vector() {constant=<rgba -1 -1 1 0>}
  n27 = AppendVector(A=ComponentMask(None=Multiply(A=@n25.RGB, B=@n26)) {g=true, r=true}, B=Multiply(A=@n25.B, B=@n4))
  n28 = StaticSwitch(True=@n9.World Space Normals XYZ, False=Normalize(VectorInput=LinearInterpolate(A=@n23, B=@n27, Alpha=@n6.Result)))
  n29 = FunctionInput(Preview=StaticBool()) {bUsePreviewValueAsDefault=true, description=This switch defaults to false., inputName=WorldSpace, inputType=FunctionInput_StaticBool}
  FunctionOutput(None=StaticSwitch(True=@n28, False=Transform(None=@n28) {transformSourceType=TRANSFORMSOURCE_World, transformType=TRANSFORM_Tangent}, Value=@n29)) {outputName=XYZ Texture, sortPriority=1}
  n30 = FunctionInput(Preview=StaticBool() {desc=Use High quality normals, value=true}) {
   bUsePreviewValueAsDefault=true
   description="Default is true. Adds ~ 20 instructions."
   inputName=Use High Quality Normals
   inputType=FunctionInput_StaticBool
   sortPriority=150
  }
  n31 = StaticSwitch(
   True=@n9.XY Flat Top Projection World Space Normals
   False=Normalize(VectorInput=LinearInterpolate(A=@n23, B=@n2, Alpha=@n6.Result))
   Value=@n30
  )
  FunctionOutput(None=StaticSwitch(True=@n31, False=Transform(None=@n31) {transformSourceType=TRANSFORMSOURCE_World, transformType=TRANSFORM_Tangent}, Value=@n29)) {outputName=XYZFlatTop, sortPriority=1}
  n32 = StaticSwitch(True=@n9.XY Projection World Space Normals, False=Normalize(VectorInput=@n23), Value=@n30)
  FunctionOutput(None=StaticSwitch(True=@n32, False=Transform(None=@n32) {transformSourceType=TRANSFORMSOURCE_World, transformType=TRANSFORM_Tangent}, Value=@n29)) {outputName=XY Texture, sortPriority=1}
  n33 = StaticSwitch(True=@n9.Z Projection World Space Normals, False=Normalize(VectorInput=@n27), Value=@n30)
  FunctionOutput(None=StaticSwitch(True=@n33, False=Transform(None=@n33) {transformSourceType=TRANSFORMSOURCE_World, transformType=TRANSFORM_Tangent}, Value=@n29)) {outputName=Z Texture, sortPriority=1}
  n34 = TextureSample(UVs=@n15, Tex=@n1) {samplerSource=SSM_Wrap_WorldGroupSettings}
  n35 = AppendVector(A=@n34.RGB, B=@n34.RGB)
  n36 = FunctionInput(Preview=StaticBool()) {bUsePreviewValueAsDefault=true, description="Default = False", inputName=Export Float 4, inputType=FunctionInput_StaticBool}
  n37 = TextureSample(UVs=@n19, Tex=@n1) {samplerSource=SSM_Wrap_WorldGroupSettings}
  n38 = AppendVector(A=@n37.RGB, B=@n37.RGB)
  n39 = MaterialFunctionCall(In=Abs(None=ComponentMask(None=@n2) {r=true}), Contrast=@n5) {materialFunction=&CheapContrast}
  n40 = TextureSample(UVs=@n24, Tex=@n1) {samplerSource=SSM_Wrap_WorldGroupSettings}
  n41 = Abs(None=ComponentMask(None=@n2) {b=true})
  LinearInterpolate(
   A=LinearInterpolate(
    A=StaticSwitch(True=@n35, False=@n34.RGB, Value=@n36)
    B=StaticSwitch(True=@n38, False=@n37.RGB, Value=@n36)
    Alpha=ComponentMask(None=@n39.Result) {r=true}
   ) {desc=Float3}
   B=StaticSwitch(True=AppendVector(A=@n40.RGB, B=@n40.RGB), False=@n40.RGB, Value=@n36)
   Alpha=ComponentMask(None=MaterialFunctionCall(In=@n41, Contrast=@n5) {materialFunction=&CheapContrast}.Result) {r=true}
  )
  WorldPosition() {worldPositionShaderOffset=WPT_ExcludeAllShaderOffsets}
 ]
}
