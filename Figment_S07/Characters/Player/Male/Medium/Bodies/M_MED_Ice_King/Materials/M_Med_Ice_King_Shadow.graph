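# Unlit translucent material for the Ice King "shadow" skeletal character, a smoky, glowing ice figure that fades via MPC_IceKing_Shadow.
# Emissive fakes an interior core with Fresnel on distorted pre-skinned position, pulsed by noise and masked.
# Opacity height-lerps panning ice and storm-cloud textures, with custom-depth self-sorting and a world-height fade.
# World position offset adds sine waves and pushes vertices outward on fadeout; feature-level switches simplify lower-end platforms.
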Material /Figment_S07/Characters/Player/Male/Medium/Bodies/M_MED_Ice_King/Materials/M_Med_Ice_King_Shadow
graph {
 properties {
  allowTranslucentCustomDepthWrites=true
  bUsedWithMeshDeformer=true
  bUsedWithSkeletalMesh=true
  blendMode=BLEND_Translucent
  pixelDepthOffsetMode=PDOM_Legacy
  shadingModel=MSM_Unlit
  usedShadingModels=Unlit
 }
 nodes [
  n1 = WorldPosition()
  n2 = Multiply(A=Time(), B=ScalarParameter() {defaultValue=-0.2, parameterName=NoisePannerSpeed})
  Add(
   A=Divide(
    A=MaterialFunctionCall(Float3=@n1.XYZ) {materialFunction=&BreakOutFloat3Components}.B
    B=ScalarParameter() {defaultValue=9625.9121, parameterName=NoisePannerScale}
   )
   B=@n2
  )
  Add()
  Constant() {r=1}
  n3 = ComponentMask(None=SceneTexture() {sceneTextureId=PPI_CustomDepth}.Color) {r=true}
  n4 = ScalarParameter() {defaultValue=-2199.7756, parameterName=BiasSorting_Far}
  n5 = Add(A=PixelDepth().R, B=@n4)
  Divide(A=Subtract(A=@n3, B=@n5), B=ScalarParameter() {defaultValue=3963.7151, parameterName=Divisor})
  n6 = VertexNormalWS()
  n7 = ScalarParameter() {defaultValue=8, parameterName=FadeoutWPO2}
  n8 = ScalarParameter() {defaultValue=29.3124, parameterName=SineWaveLength}
  n9 = Divide(A=ComponentMask(None=PreSkinnedPosition()) {b=true}, B=@n8)
  n10 = Add(A=Sine(None=@n9)) {constB=1.1}
  n11 = Multiply(
   A=VectorParameter() {parameterName=DirectionPush, defaultValue=<rgba 1 1 0 1>}.RGB
   B=ScalarParameter() {defaultValue=22536.3906, parameterName=FadeoutWPO}
  )
  n12 = CollectionParameter() {collection=&/MPC_IceKing_Shadow, parameterName=Fadeout}
  n13 = LinearInterpolate(B=Add(A=Multiply(A=@n6, B=@n7), B=Multiply(A=@n10, B=@n11)), Alpha=@n12)
  n14 = Time()
  n15 = ScalarParameter() {defaultValue=-0.3152, parameterName=SpeedWave}
  n16 = Add(A=@n9, B=Frac(None=Multiply(A=@n14, B=@n15)))
  n17 = ScalarParameter() {defaultValue=0.4, parameterName=PushDistance}
  n18 = MaterialFunctionCall() {materialFunction=/Engine/Functions/Engine_MaterialFunctions02/WorldPositionOffset/ObjectScale}
  n19 = Add(A=@n13, B=Multiply(A=Sine(None=@n16), B=Multiply(A=@n17, B=@n18.Scale Z)))
  FeatureLevelSwitch(Default=@n19, ES2=@n13, ES3_1=@n13, SM4_REMOVED=@n19)
  n20 = Constant()
  n21 = Clamp(None=Add(
    A=CollectionParameter() {collection=&MPC_IceKing_Shadow, parameterName=FadeIn}
    B=@n12
   ))
  n22 = OneMinus(None=@n21)
  n23 = Fresnel(ExponentIn=LinearInterpolate(
    A=ScalarParameter() {defaultValue=0.1951, parameterName=FresnelExp}
    B=ScalarParameter() {defaultValue=-0.9789, parameterName=FresnelExp2}
    Alpha=@n21
   )) {baseReflectFraction=0}
  n24 = OneMinus(None=@n23)
  n25 = ScalarParameter() {defaultValue=29.7352, parameterName=IceTextureScale}
  n26 = AppendVector(A=@n25, B=@n25)
  n27 = Transform(None=VertexNormalWS()) {transformSourceType=TRANSFORMSOURCE_World, transformType=TRANSFORM_Local}
  n28 = ScalarParameter() {defaultValue=71.5129, parameterName=SmokeTextureScale}
  n29 = AppendVector(A=AppendVector(A=@n28, B=@n28), B=@n28)
  n30 = AppendVector(A=ScalarParameter() {defaultValue=0.05, parameterName=PanSpeedX}, B=ScalarParameter() {defaultValue=0.0667, parameterName=PanSpeedY})
  n31 = MaterialFunctionCall!(
   TextureObject=TextureObject() {samplerType=SAMPLERTYPE_LinearGrayscale, texture=/Game/VisualThreatManager/StormVisuals/LargeClouds/T_StormCloudDetail01_M}
   TextureSize=@n29
   WorldPosition
   Export Float 4
   In=@n30
   World Space Normal=@n27
   ProjectionTransitionContrast
  ) {materialFunction=/Game/Athena/SafeZone/Preskinned_WP_Texture_ScalePan}
  n32 = Add(
   A=MaterialFunctionCall(
    TextureObject=TextureObject() {texture=/Game/Packages/Fortnite_UniqueMaterials_1/Textures/T_Ice_Lake_Snow_D}
    TextureSize=AppendVector(A=@n26, B=@n25)
    World Space Normal=@n27
   ) {materialFunction=&Preskinned_WP_Texture_ScalePan}.XY Texture
   B=@n31.XY Texture
  )
  n33 = FeatureLevelSwitch(Default=@n32, ES2=Multiply(A=@n31.XY Texture) {constB=2}, ES3_1=@n32, SM4_REMOVED=@n32)
  n34 = MaterialFunctionCall(
   Transition Phase=Clamp(None=Subtract(A=@n24, B=@n21))
   Height Texture=@n33
   Contrast=ScalarParameter() {defaultValue=0.0274, parameterName=ContrastHeightLerp}
  ) {materialFunction=&HeightLerp}
  n35 = Multiply(
   A=If(A=@n5, B=@n3, "A > B"=@n20, "A == B"=@n20, "A < B"=Constant() {r=1})
   B=Multiply(A=ScalarParameter() {defaultValue=0.7, parameterName=Opacity}, B=Multiply(A=@n22, B=@n34.Alpha))
  )
  n36 = Constant() {r=1}
  n37 = ComponentMask(None=WorldPosition().XYZ) {b=true}
  n38 = ScalarParameter() {defaultValue=-24604.0859, parameterName=OpacityFadeBottomOffset}
  n39 = Clamp(None=Divide(A=Add(A=@n37, B=@n38), B=ScalarParameter() {defaultValue=135652.7344, parameterName=OpacityFadeBottomDistance}))
  n40 = CollectionParameter() {collection=&MPC_IceKing_Shadow, parameterName=FadeoutBottom}
  n41 = Multiply(A=@n35, B=VertexInterpolator(VS=LinearInterpolate(A=@n36, B=@n39, Alpha=@n40)).PS)
  FeatureLevelSwitch(Default=@n41, ES2=@n35, ES3_1=@n35, SM4_REMOVED=@n41)
  LinearInterpolate(
   A=ScalarParameter() {defaultValue=-2660.3965, parameterName=BiasSorting_Near}
   B=@n4
   Alpha=MaterialFunctionCall(Distance=ScalarParameter() {defaultValue=100000, parameterName=BiasDistancefade}) {materialFunction=&Near_Camera_Fade}.Result
  )
  MaterialFunctionCall() {materialFunction=&CameraDepthFade}
  n42 = VectorParameter() {parameterName=Skylight, defaultValue=<rgba 0.9844 0.9888 1 1>}
  n43 = CollectionParameter() {collection=&FortniteMaterialParameters, parameterName=SunLightColor}
  n44 = Multiply(
   A=VectorParameter() {parameterName=SunlightTint, defaultValue=<rgba 1.1 1.1 1.1 1>}.RGB
   B=Desaturation(None=ComponentMask(None=@n43) {b=true, g=true, r=true}, Fraction=ScalarParameter() {defaultValue=0.5, parameterName=DesatReallight})
  )
  n45 = CollectionParameter() {
   collection=&FortniteMaterialParameters
   parameterName=SunlightDirectionalVector
  }
  n46 = ComponentMask(None=@n45) {b=true, g=true, r=true}
  n47 = TextureSampleParameter2D() {
   parameterName=Normal
   samplerType=SAMPLERTYPE_Normal
   texture=/BRCosmetics/Characters/Player/Male/Medium/Bodies/M_MED_BriskReign/Textures/T_BriskReign_Body_N
  }
  n48 = MaterialFunctionCall!(
   Export Float 4
   TextureObject=TextureObject() {samplerType=SAMPLERTYPE_Normal, texture=/Game/VisualThreatManager/StormVisuals/LargeClouds/T_StormCloudDetail01_N}
   TextureSize=@n29
   WorldPosition
   World Space Normal=@n27
   In=@n30
   WorldSpace
   ProjectionTransitionContrast
   Use High Quality Normals
  ) {materialFunction=/Figment_S07/Athena/SafeZone/Preskinned_WP_normal_Texture_ScalePan1}
  n49 = ScalarParameter() {defaultValue=0.4095, parameterName=FlattenSmokeNormal}
  n50 = Transform(None=MaterialFunctionCall(
    BaseNormal=@n47.RGB
    AdditionalNormal=MaterialFunctionCall(Normal=@n48.XYZ Texture, Flatness=@n49) {materialFunction=&FlattenNormal}.Result
   ) {materialFunction=&BlendAngleCorrectedNormals}.Result)
  n51 = DotProduct(A=@n46, B=@n50)
  n52 = Clamp(None=Power(Base=Clamp(None=@n51) {clampMode=CMODE_ClampMin}, Exp=ScalarParameter() {defaultValue=1, parameterName=FakelLightPower}))
  n53 = VectorParameter() {parameterName=SpecularColor, defaultValue=<rgba 0.3538 0.357 0.3594 1>}
  n54 = Normalize(VectorInput=Add(A=@n46, B=CameraVectorWS()))
  n55 = Power(
   Base=Clamp(None=DotProduct(A=@n50, B=@n54)) {clampMode=CMODE_ClampMin}
   Exp=ScalarParameter() {defaultValue=1.1243, parameterName=FakelSpecPower_1}
  )
  Multiply(
   A=ScalarParameter() {defaultValue=0.58, parameterName=BrightnessFakeLight}
   B=Add(A=LinearInterpolate(A=@n42.RGB, B=@n44, Alpha=@n52), B=Multiply(A=@n53.RGB, B=@n55))
  )
  n56 = PreSkinnedPosition()
  n57 = ScalarParameter() {defaultValue=38.2283, parameterName=DistortionICE}
  n58 = Multiply(
   A=VectorParameter() {parameterName=Stretchdistortion, defaultValue=<rgba 1 1 2 1>}.RGB
   B=LinearInterpolate(A=@n57, B=Multiply(A=@n57) {constB=-1}, Alpha=ComponentMask(None=@n33) {r=true})
  )
  n59 = ScalarParameter() {parameterName=BiasTexture}
  n60 = TextureSampleParameter2D!(UVs, Bias=@n59, Apply View MipBias) {
   mipValueMode=TMVM_MipBias
   parameterName=Normal
   samplerType=SAMPLERTYPE_Normal
   texture=&T_BriskReign_Body_N
  }
  n61 = Multiply(A=ScalarParameter() {defaultValue=157.7281, parameterName=DistortionICE2}, B=Transform(None=@n60.RGB))
  n62 = ScalarParameter() {parameterName=CenterOffsetX}
  n63 = ScalarParameter() {parameterName=CenterOffsetY}
  n64 = Add(
   A=Add(A=VertexInterpolator(VS=@n56).PS, B=Add(A=@n58, B=@n61))
   B=AppendVector(A=AppendVector(A=@n62, B=@n63), B=ScalarParameter() {defaultValue=-114.3519, parameterName=CenterOffsetZ})
  )
  n65 = Multiply(A=CameraVectorWS(), B=ScalarParameter() {defaultValue=68.5128, parameterName=DistanceBehind Object})
  n66 = Fresnel(
   ExponentIn=ScalarParameter() {defaultValue=1.1471, parameterName=IceFalloffPowerEmissiveCore}
   BaseReflectFractionIn=ScalarParameter() {defaultValue=0.2, parameterName=IceFalloffPowerEmissiveCoreBase}
   Normal=Normalize(VectorInput=Add(A=@n64, B=@n65))
  ) {baseReflectFraction=0}
  n67 = Multiply(A=Power(Base=@n66, Exp=Constant() {r=2}), B=VectorParameter() {parameterName=EmissiveCore, defaultValue=<rgba 0.3305 1.5 1.0557 1>}.RGB)
  n68 = Power(Base=OneMinus(None=@n66), Exp=ScalarParameter() {defaultValue=25.6681, parameterName=IceFalloffPowerEmissiveCore2})
  n69 = ScalarParameter() {defaultValue=26.7151, parameterName=CoreEmissive}
  n70 = Add(A=@n67, B=Multiply(A=VectorParameter() {parameterName=EmissiveCore2, defaultValue=<rgba 0.4564 2 1.4502 1>}.RGB, B=Multiply(A=@n68, B=@n69)))
  n71 = Multiply(
   A=@n70
   B=Add(
    A=ScalarParameter() {defaultValue=0.7785, parameterName=PulseMin}
    B=TextureSample(UVs=@n2) {texture=/Figment_S07/Athena/Environments/DudeBro/T_Dudebro_Noise_01}.G
   )
  )
  n72 = VectorParameter() {parameterName=Color_1, defaultValue=<rgba 0.0035 0.0216 0.0417 1>}
  n73 = VectorParameter() {parameterName=Color_2, defaultValue=<rgba 0.0072 0.0324 0.1562 1>}
  n74 = TextureSampleParameter2D(Bias=@n59) {
   mipValueMode=TMVM_MipBias
   parameterName=M
   samplerType=SAMPLERTYPE_Masks
   texture=/BRCosmetics/Characters/Player/Male/Medium/Bodies/M_MED_BriskReign/Textures/T_BriskReign_Body_M
  }
  n75 = LinearInterpolate(
   A=Multiply(A=@n67, B=ScalarParameter() {defaultValue=0.6555, parameterName=HoodInsideDarken})
   B=Add(A=FeatureLevelSwitch(Default=@n71, ES2=@n70, ES3_1=@n70, SM4_REMOVED=@n71), B=LinearInterpolate(A=@n72.RGB, B=@n73.RGB, Alpha=@n23))
   Alpha=@n74.R
  )
  n76 = Divide(A=ComponentMask(None=PreSkinnedPosition()) {b=true}, B=ScalarParameter() {defaultValue=156.1487, parameterName=TopdownGradient})
  n77 = LinearInterpolate(A=Desaturation(None=@n75), B=@n75, Alpha=VertexInterpolator(VS=Clamp(None=@n76)).PS)
  Multiply(
   A=Multiply(A=FeatureLevelSwitch(Default=@n77, ES2=@n75, ES3_1=@n75, SM4_REMOVED=@n77), B=OneMinus(None=@n21))
   B=ScalarParameter() {defaultValue=2, parameterName=BrightnessAdjust}
  )
  Multiply(A=@n8, B=@n18.Scale Z)
  ScalarParameter() {parameterName=FadeIn}
  ScalarParameter() {parameterName=OpacityFadeBottom}
  VertexNormalWS()
 ]
}
